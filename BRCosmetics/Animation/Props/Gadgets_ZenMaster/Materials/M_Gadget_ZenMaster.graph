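# Masked, two-sided material for the Zen Master gadget cosmetic that dissolves it with pink petal and glowing edge effects.
# It inserts custom opacity mask and emissive between the shared CharacterParent PreFX and PostFX functions.
# Several sphere masks, with a DisappearRadius parameter and a noise-perturbed ramp, define top, bottom and edge regions.
# Three panning petal-texture layers are blended, alongside pink color gradients and an HDR EdgeColor for the emissive.
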Material /BRCosmetics/Animation/Props/Gadgets_ZenMaster/Materials/M_Gadget_ZenMaster
graph {
 properties {
  bUseMaterialAttributes=true
  bUsedWithClothing=true
  bUsedWithMeshDeformer=true
  bUsedWithMorphTargets=true
  bUsedWithNanite=true
  bUsedWithNiagaraMeshParticles=true
  bUsedWithSkeletalMesh=true
  blendMode=BLEND_Masked
  ditherOpacityMask=true
  pixelDepthOffsetMode=PDOM_Legacy
  shadingModel=MSM_FromMaterialExpression
  twoSided=true
  usedShadingModels=Default Lit | Subsurface | Clear Coat | Subsurface Profile | Cloth
 }
 nodes [
  Constant()
  Constant()
  n1 = MaterialFunctionCall() {materialFunction=&MF_CharacterParent_2023_PreFX}
  MaterialFunctionCall(
   Material=SetMaterialAttributes!(MaterialAttributes=@n1.Result, Opacity Mask=NamedRerouteUsage(), Emissive Color=NamedRerouteUsage()) {attributeSetTypes [#1, #2]}
   M=@n1.M
   Pre Skinned Local Position=@n1.M
   Distance Blend=@n1.M
   Fresnel=@n1.M
  ) {materialFunction=&MF_CharacterParent_2023_PostFX}
  n2 = ScalarParameter() {group=ZenMaster, parameterName=DisappearRadius}
  Max(A=ScalarParameter() {group=ZenMaster, parameterName=DisapperRadius}, B=@n2)
  n3 = ComponentMask(None=NamedRerouteUsage()) {r=true}
  n4 = Add(A=OneMinus(None=NamedRerouteUsage()), B=Min(A=NamedRerouteUsage(), B=Saturate(None=@n3)))
  NamedRerouteDeclaration(None=Saturate(None=Multiply(A=@n4) {constB=2})) {name=OPacityMask, nodeColor=<rgba 0 0.7656 1 1>}
  n5 = TextureSample(UVs=Multiply(A=TextureCoordinate(), B=Constant() {r=5})) {texture=&T_NoisePack_01}
  n6 = Constant() {r=0.03}
  NamedRerouteDeclaration(None=Add(A=TextureSample() {texture=/BRCosmetics/Animation/Props/Gadgets_ZenMaster/Textures/T_ramp_Mask}.R, B=Multiply(A=@n5.R, B=@n6))) {name=WP NOise, nodeColor=<rgba 0 1 0.5078 1>}
  n7 = NamedRerouteUsage()
  n8 = NamedRerouteUsage()
  n9 = NamedRerouteUsage()
  n10 = ScalarParameter() {defaultValue=0.65, group=ZenMaster, parameterName=Hardness}
  NamedRerouteDeclaration(None=OneMinus(None=SphereMask(A=@n7, B=@n8, Radius=@n9, Hardness=@n10) {hardnessPercent=2})) {name=Bottom mask, nodeColor=<rgba 1 0.8906 0 1>}
  n11 = TextureCoordinate()
  n12 = TextureSample(UVs=Multiply(A=Panner(Coordinate=@n11) {speedX=0.01, speedY=-0.1}) {constB=2}) {samplerSource=SSM_Wrap_WorldGroupSettings, texture=/BRCosmetics/Animation/Props/Gadgets_ZenMaster/Textures/T_petal_Mask}
  n13 = Panner(Coordinate=TextureCoordinate()) {speedX=-0.02, speedY=-0.12}
  n14 = Multiply(A=TextureSample(UVs=Multiply(A=@n13) {constB=3}) {samplerSource=SSM_Wrap_WorldGroupSettings, texture=&T_petal_Mask}.RGB) {constB=0.8}
  n15 = TextureCoordinate()
  n16 = TextureSample(UVs=Multiply(A=Panner(Coordinate=@n15) {speedX=0.02, speedY=-0.15}) {constB=5}) {samplerSource=SSM_Wrap_WorldGroupSettings, texture=&T_petal_Mask}
  NamedRerouteDeclaration(None=LinearInterpolate(A=LinearInterpolate(A=@n12.RGB, Alpha=@n14), Alpha=Multiply(A=@n16.RGB) {constB=0.5})) {name=petal, nodeColor=<rgba 1 0.7266 0 1>}
  NamedRerouteDeclaration(None=@n2) {name=Animation, nodeColor=<rgba 0.8984 1 0 1>}
  n17 = Constant()
  NamedRerouteDeclaration(None=SphereMask(
    A=NamedRerouteUsage()
    B=NamedRerouteDeclaration(None=@n17) {name=Position, nodeColor=<rgba 1 0.5625 0 1>}
    Radius=NamedRerouteUsage()
    Hardness=ScalarParameter() {defaultValue=0.98, parameterName=Hardness2}
   ) {hardnessPercent=2}) {name=Top mask, nodeColor=<rgba 0.5703 0 1 1>}
  n18 = NamedRerouteUsage()
  n19 = ScalarParameter() {defaultValue=-0.01, parameterName=radius offset}
  NamedRerouteDeclaration(None=SphereMask(
    A=NamedRerouteUsage()
    B=NamedRerouteUsage()
    Radius=Add(A=@n18, B=@n19) {constB=-0.5}
    Hardness=ScalarParameter() {defaultValue=0.99, group=ZenMaster, parameterName=Hardness4}
   ) {hardnessPercent=2}) {name=Highlighted edges, nodeColor=<rgba 0 1 0.25 1>}
  n20 = ComponentMask(None=NamedRerouteUsage()) {b=true}
  n21 = LinearInterpolate(
   A=Constant3Vector() {constant=<rgba 0.98 0.0196 0.4107 1>}
   B=Constant3Vector() {constant=<rgba 1 0.3372 0.6172 1>}
   Alpha=ComponentMask(None=NamedRerouteUsage()) {g=true}
  )
  n22 = Multiply(A=ComponentMask(None=NamedRerouteUsage()) {r=true}) {constB=3}
  NamedRerouteDeclaration(None=Add(A=Multiply(A=@n20, B=@n21) {constB=0.2}, B=Multiply(A=@n21, B=@n22))) {name=EmissiveColor Colot Out, nodeColor=<rgba 1 0 0.0703 1>}
  NamedRerouteDeclaration(None=SphereMask(
    A=NamedRerouteUsage()
    B=NamedRerouteUsage()
    Radius=NamedRerouteUsage()
    Hardness=ScalarParameter() {defaultValue=0.98, group=ZenMaster, parameterName=Hardness5}
   ) {hardnessPercent=2}) {name=Edge mask_1, nodeColor=<rgba 0.5703 0 1 1>}
  n23 = NamedRerouteUsage()
  n24 = Multiply(A=@n23, B=OneMinus(None=@n23))
  n25 = VectorParameter() {group=ZenMaster, parameterName=EdgeColor, defaultValue=<rgba 30 3.3036 26.2867 1>}
  n26 = NamedRerouteUsage()
  n27 = NamedRerouteUsage()
  n28 = Add(A=Multiply(A=Subtract(A=@n26, B=@n27)) {constB=3}, B=Multiply(A=NamedRerouteUsage(), B=NamedRerouteUsage()))
  n29 = NamedRerouteUsage()
  NamedRerouteDeclaration(None=Add(A=Multiply(A=@n24, B=@n25.RGB), B=Multiply(A=@n28, B=@n29))) {name=Emissive Out, nodeColor=<rgba 0 1 0.5547 1>}
 ]
}
